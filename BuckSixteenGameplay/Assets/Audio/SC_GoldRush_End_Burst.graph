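# Sound cue for the burst at the end of the Gold Rush buff countdown.
# A Random node picks one of four countdown-end wave variations with equal weight and feeds it through a Modulator at default settings.
# Playback uses the Athena consumable-use attenuation asset and the consumables sound class.
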SoundCue /BuckSixteenGameplay/Assets/Audio/SC_GoldRush_End_Burst
root=Modulator_0
nodes [
 {id=Modulator_0, children [Random_1]}
 {id=Random_1, children [WavePlayer_0,WavePlayer_1,WavePlayer_2,WavePlayer_3], properties {weights [1,1,1,1]}}
 {id=WavePlayer_0, properties {soundWaveAssetPtr=&/SC_GoldRush_Buff_Countdown_End_01}}
 {id=WavePlayer_1, properties {soundWaveAssetPtr=&/SC_GoldRush_Buff_Countdown_End_02}}
 {id=WavePlayer_2, properties {soundWaveAssetPtr=&/SC_GoldRush_Buff_Countdown_End_03}}
 {id=WavePlayer_3, properties {soundWaveAssetPtr=&/SC_GoldRush_Buff_Countdown_End_04}}
]
properties {
 attenuationSettings=&Fort_Consumable_Use_Athena_Att
 maxDistance=2250
 soundClassObject=&Fort_Consumables
 volumeMultiplier=1
}
