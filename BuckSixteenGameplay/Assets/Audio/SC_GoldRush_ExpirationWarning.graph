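# Sound cue for the Gold Rush buff's expiration warning, playing a countdown sound as the buff runs out.
# A Random node picks among four equally weighted countdown waves and feeds a Modulator for pitch and volume modulation, with the pitch floor raised to normal.
# The cue plays at reduced volume and uses the shared consumable-use attenuation asset for distance falloff.
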SoundCue /BuckSixteenGameplay/Assets/Audio/SC_GoldRush_ExpirationWarning
root=Modulator_0
nodes [
 {id=Modulator_0, children [Random_0], properties {pitchMin=1}}
 {id=Random_0, children [WavePlayer_0,WavePlayer_1,WavePlayer_2,WavePlayer_3], properties {weights [1,1,1,1]}}
 {id=WavePlayer_0, properties {soundWaveAssetPtr=&/GoldRush_Buff_Countdown_01}}
 {id=WavePlayer_1, properties {soundWaveAssetPtr=&/GoldRush_Buff_Countdown_02}}
 {id=WavePlayer_2, properties {soundWaveAssetPtr=&/GoldRush_Buff_Countdown_03}}
 {id=WavePlayer_3, properties {soundWaveAssetPtr=&/GoldRush_Buff_Countdown_04}}
]
properties {attenuationSettings=&Fort_Consumable_Use_Athena_Att, maxDistance=2250, volumeMultiplier=0.5}
