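# Unlit translucent Niagara ribbon material for the Gold Rush buff's sand trail.
# Two scales of panning sand noise, multiplied and contrasted, boost emissive and opacity.
# Opacity comes from a wind-normal-distorted edge mask minus panning erosion, times a trail texture, with near-camera fade and depth fade (skipped on mobile).
# Particle color tints and fades it, while dynamic parameters offset time, scale sand and mirror the edge mask.
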Material /BuckSixteenGameplay/Assets/FX/M_GoldRush_Buff_SandTrail
graph {
 properties {
  bEnableResponsiveAA=true
  bUsedWithNiagaraRibbons=true
  blendMode=BLEND_Translucent
  pixelDepthOffsetMode=PDOM_Legacy
  shadingModel=MSM_Unlit
  usedShadingModels=Unlit
 }
 nodes [
  n1 = VectorParameter() {parameterName=EmissiveColor, defaultValue=<rgba 1 1 1 1>}
  n2 = VectorParameter() {parameterName=SandSettings, channelNames {a=PanY, b=PanX, g=ScaleY, r=ScaleX}, defaultValue=<rgba 3 0.3 -1.75 0>}
  n3 = Multiply(A=TextureCoordinate(), B=AppendVector(A=@n2.ScaleX, B=@n2.ScaleX))
  n4 = DynamicParameter() {paramNames [TimeOffset,SandOpacity,LeftRight,Param4]}
  n5 = Add(A=Time(), B=@n4.TimeOffset)
  n6 = AppendVector(A=@n2.PanX, B=@n2.PanX)
  n7 = Divide(A=@n3, B=Constant() {r=0.5})
  n8 = Multiply(
   A=TextureSample(UVs=Panner(Coordinate=@n3, Time=@n5, Speed=@n6) {speedX=-0.5}) {samplerType=SAMPLERTYPE_LinearGrayscale, texture=/Game/Global/Textures/Default/Noise/T_SandTime2}.R
   B=TextureSample(UVs=Panner(Coordinate=@n7, Time=@n5, Speed=@n6) {speedX=-0.5}) {samplerType=SAMPLERTYPE_LinearGrayscale, texture=&T_SandTime2}.R
  )
  n9 = Constant() {r=1}
  n10 = Multiply(
   A=Saturate(None=MaterialFunctionCall(In=@n8, Contrast=@n9) {materialFunction=&CheapContrast}.Result)
   B=Multiply(A=ScalarParameter() {defaultValue=5, parameterName=GoldSandIntensity}, B=@n4.SandOpacity)
  )
  n11 = Divide(A=Multiply(A=@n1.RGB, B=@n10)) {constB=3}
  n12 = ParticleColor()
  MaterialFunctionCall(Emissive=Multiply(A=Add(A=@n1.RGB, B=@n11), B=@n12.RGB)) {materialFunction=&MF_FX_EmissiveAdjustments}
  Multiply()
  n13 = VectorParameter() {parameterName=NoiseUVs, channelNames {a=PanY, b=PanX, g=ScaleY, r=ScaleX}, defaultValue=<rgba 0.7 0.1 -1.25 0>}
  n14 = Time()
  n15 = DynamicParameter() {paramNames [TimeOffset,SandOpacity,LeftRight,Param4]}
  n16 = Add(
   A=TextureCoordinate()
   B=Multiply(
    A=MaterialFunctionCall(X=@n13.PanX, Y=@n13.PanX) {materialFunction=&MakeFloat2}.Result
    B=Add(A=@n14, B=@n15.TimeOffset)
   )
  )
  n17 = MaterialFunctionCall(X=@n13.ScaleX, Y=@n13.ScaleX) {materialFunction=&MakeFloat2}
  n18 = ComponentMask(None=TextureSample(UVs=Multiply(A=@n16, B=@n17.Result)) {samplerType=SAMPLERTYPE_Normal, texture=/Game/Effects/Fort_Effects/Textures/Tile/T_Wind_N}.RGB) {g=true, r=true}
  n19 = ScalarParameter() {defaultValue=0.3, parameterName=DistoStrength}
  n20 = Multiply(A=@n18, B=@n19) {constB=0}
  n21 = TextureCoordinate()
  n22 = Add(
   A=@n20
   B=LinearInterpolate(A=OneMinus(None=@n21), B=@n21, Alpha=DynamicParameter() {paramNames [TimeOffset,SandOpacity,LeftRight,Param4]}.LeftRight)
  )
  n23 = ComponentMask(None=@n21) {r=true}
  n24 = Multiply(
   A=SmoothStep(Value=ComponentMask(None=@n22) {g=true}) {constMax=0.8}
   B=Multiply(A=SmoothStep(Value=@n23) {constMax=0.05}, B=SmoothStep(Value=@n23) {constMax=0, constMin=1})
  )
  n25 = VectorParameter() {parameterName=ErosionUVs, channelNames {a=PanY, b=PanX, g=ScaleY, r=ScaleX}, defaultValue=<rgba 3 2 -2 0.8>}
  n26 = Time()
  n27 = DynamicParameter() {paramNames [TimeOffset,SandOpacity,LeftRight,Param4]}
  n28 = Add(
   A=TextureCoordinate()
   B=Multiply(
    A=MaterialFunctionCall(X=@n25.PanX, Y=@n25.PanX) {materialFunction=&MakeFloat2}.Result
    B=Add(A=@n26, B=@n27.TimeOffset)
   )
  )
  n29 = MaterialFunctionCall(X=@n25.ScaleX, Y=@n25.ScaleX) {materialFunction=&MakeFloat2}
  n30 = TextureSample(UVs=Add(A=Multiply(A=@n28, B=@n29.Result), B=@n20)) {texture=/Game/Effects/Fort_Effects/Textures/Tile/T_Smooth_Tile_Chromatic}
  n31 = ComponentMask(None=TextureCoordinate()) {r=true}
  n32 = Multiply(
   A=Multiply(A=SmoothStep(Value=@n30.R) {constMax=0.7}, B=SmoothStep(Value=@n31) {constMax=0.5})
   B=ScalarParameter() {defaultValue=3, parameterName=ErosionStrength}
  )
  n33 = Multiply(A=@n18, B=@n19) {constB=0.1}
  n34 = TextureCoordinate()
  n35 = VectorParameter() {parameterName=TrailBaseUVs, channelNames {a=PanY, b=PanX, g=ScaleY, r=ScaleX}, defaultValue=<rgba 1 1 -1.7 0>}
  n36 = MaterialFunctionCall(X=@n35.PanX, Y=@n35.PanX) {materialFunction=&MakeFloat2}
  n37 = Multiply(
   A=Add(A=Add(A=@n33, B=@n34), B=Multiply(A=@n36.Result, B=@n5))
   B=MaterialFunctionCall(X=@n35.ScaleX, Y=@n35.ScaleX) {materialFunction=&MakeFloat2}.Result
  )
  n38 = Multiply(
   A=Saturate(None=Subtract(A=@n24, B=@n32))
   B=Power(
    Base=TextureSample(UVs=@n37) {texture=/Game/Weapons/FORT_Melee/Pickaxe_Tower_Sentinel_Female/FX/T_TowerSentinel_trail}.G
    Exp=ScalarParameter() {defaultValue=0.4, parameterName=TexturePower}
   ) {constExponent=0.5}
  )
  n39 = Add(A=Multiply(A=@n38, B=@n10), B=@n38)
  n40 = DepthFade(Opacity=@n39, FadeDistance=ScalarParameter() {defaultValue=40, parameterName=Depthfade})
  Multiply(
   A=Multiply(A=@n12.A, B=ShadingPathSwitch(Default=@n40, Deferred=@n40, Mobile=@n39) {desc=Migrated from FeatureLevelSwitch})
   B=MaterialFunctionCall(Distance=ScalarParameter() {defaultValue=60, parameterName=NearCameraFade}) {materialFunction=&Near_Camera_Fade}.Result
  )
 ]
}
