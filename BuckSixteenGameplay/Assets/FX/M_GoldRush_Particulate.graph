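# Unlit translucent particle material for Niagara sprites and meshes that produces an animated sparkle, apparently for a Gold Rush effect.
# Two tile-noise samples, scaled about UV center and panned at different velocities, multiply into a shimmering mask.
# Dynamic parameters drive per-particle UV scale, random offset, pan speed and contrast.
# The contrasted mask is tinted by particle color and fed into the shared emissive adjustment function.
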Material /BuckSixteenGameplay/Assets/FX/M_GoldRush_Particulate
graph {
 properties {
  bAutomaticallySetUsageInEditor=false
  bUsedWithNiagaraMeshParticles=true
  bUsedWithNiagaraSprites=true
  blendMode=BLEND_Translucent
  shadingModel=MSM_Unlit
  usedShadingModels=Unlit
 }
 nodes [
  Add() {constB=0.5}
  n1 = ParticleColor()
  n2 = TextureCoordinate()
  n3 = DynamicParameter() {defaultValue=<rgba 0.03 1 0 1>, paramNames [SparkleContrast, SparkleUVScale, SparkleRand, SparkleUVSpeed]}
  n4 = ScalarParameter() {defaultValue=4, parameterName=SparkleUVScale1}
  n5 = Panner(
   Coordinate=MaterialFunctionCall(UVs=Add(A=@n2, B=@n3.SparkleRand), Texture Scale=Multiply(A=@n4, B=@n3.SparkleUVScale) {constA=-1}) {materialFunction=&ScaleUVsByCenter}.UVs
   Speed=Multiply(A=Constant2Vector() {g=0.3, r=0.2}, B=@n3.SparkleUVSpeed) {constA=-1}
  )
  n6 = MaterialFunctionCall(
   UVs=Add(A=TextureCoordinate(), B=@n3.SparkleRand)
   Texture Scale=Multiply(A=ScalarParameter() {defaultValue=1, parameterName=SparkleUVScale2}, B=@n3.SparkleUVScale) {constA=-1}
  ) {materialFunction=&ScaleUVsByCenter}
  n7 = Multiply(A=Constant2Vector() {g=-0.04, r=0.12}, B=@n3.SparkleUVSpeed) {constA=-1}
  n8 = Multiply(
   A=LinearInterpolate(Alpha=TextureSample(UVs=@n5) {samplerSource=SSM_Wrap_WorldGroupSettings, texture=/Game/Effects/Fort_Effects/Textures/Tile/T_TileNoise_01}.R) {constA=0.02}
   B=TextureSample(UVs=Panner(Coordinate=@n6.UVs, Speed=@n7)) {samplerSource=SSM_Wrap_WorldGroupSettings, texture=&T_TileNoise_01}.R
  ) {constA=-1}
  n9 = DynamicParameter() {defaultValue=<rgba 0.03 1 0 1>, paramNames [SparkleContrast, SparkleUVScale, SparkleRand, SparkleUVSpeed]}
  MaterialFunctionCall(Emissive=Multiply(
    A=@n1.RGB
    B=MaterialFunctionCall(In=@n8, Contrast=@n9.SparkleContrast) {materialFunction=&CheapContrast}.Result
   ) {constA=-1}) {materialFunction=&MF_FX_EmissiveAdjustments}
  ParticleRandom()
  ParticleRandom()
  Rotator(Time=Time()) {centerX=0.4}
  Rotator(Time=Time()) {centerY=0.45, speed=-0.15}
  ScalarParameter() {defaultValue=0.3, parameterName=SparkleRotSpeed2}
  ScalarParameter() {defaultValue=0.03, parameterName=SparkleContrast}
  Subtract(A=@n8, B=@n1.A)
 ]
}
