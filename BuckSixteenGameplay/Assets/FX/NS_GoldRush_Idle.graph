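# Niagara system for the Gold Rush idle effect, attaching gold glows, dust and trails to a character's skeletal mesh.
# Bokeh, Decals and Particulate sprite emitters map particle partitions to bones or sockets for sparkles, glows and falling dust.
# Four ribbons stretch opposite owner velocity by normalized speed, boosted for third-person viewers.
# System-level water opacity modulates spawning or visibility, ADS alpha adjusts sprite fading, and platform checks tune sprites.
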NiagaraSystem /BuckSixteenGameplay/Assets/FX/NS_GoldRush_Idle
dependencies {
 usedDataInterfaces [
  &NiagaraDataInterfaceArrayFloat
  &NiagaraDataInterfaceArrayFloat3
  &NiagaraDataInterfaceCurve
  &NiagaraDataInterfacePlatformSet
  &NiagaraDataInterfaceArrayInt32
  &NiagaraDataInterfaceColorCurve
  &NiagaraDataInterfaceSkeletalMesh
  &NiagaraDataInterfaceCamera
 ]
 usedDynamicInputs [
  &SetBoolByFloatComparison
  &MakeCustomVectorFromBool
  &ClampFloat
  &RemapRange
  &Subtract_Vector
  &ReturnCameraPosition
 ]
 usedModules [
  &self:SystemSpawnScript.SystemScriptSource.SystemScriptGraph.NiagaraNodeAssignment_1.SetVariables_#1
  &self:SystemSpawnScript.SystemScriptSource.SystemScriptGraph.NiagaraNodeAssignment_2.SetVariables_#2
  &self:SystemSpawnScript.SystemScriptSource.SystemScriptGraph.NiagaraNodeAssignment_5.SetVariables_#3
  &SystemState
  &Timeline
  &self:SystemSpawnScript.SystemScriptSource.SystemScriptGraph.NiagaraNodeAssignment_8.SetVariables_#4
  &InterpolateOverTime
  &self:SystemSpawnScript.SystemScriptSource.SystemScriptGraph.NiagaraNodeAssignment_3.SetVariables_#5
  &self:SystemSpawnScript.SystemScriptSource.SystemScriptGraph.NiagaraNodeAssignment_4.SetVariables_#6
  &self:SystemSpawnScript.SystemScriptSource.SystemScriptGraph.NiagaraNodeAssignment_0.SetVariables_#7
  &self:SystemSpawnScript.SystemScriptSource.SystemScriptGraph.NiagaraNodeAssignment_6.SetVariables_#8
  &self:SystemSpawnScript.SystemScriptSource.SystemScriptGraph.NiagaraNodeAssignment_10.SetVariables_#9
 ]
}
emitters {
 Bokeh {
  emitterUpdateScript [
   &EmitterState()
   &SpawnRate(Spawn Probability=0.75, SpawnRate=&Multiply_Float(A=$System.WaterFXOpacity, B=8), Use Spawn Probability=true)
  ]
  particleSpawnScript [
   &PartitionParticles(Number Of Partitions=4)
   &InitializeParticle(
    Color=&ColorFromCurve(
     ColorCurve=DataInterface() {propertyValues {
       AlphaCurve {keys [<at 0.4783 0.1 RCIM_Cubic>]}
       BlueCurve {keys [<at 0 0>,<at 0.3901 0>,<at 0.7425 0.118>,<at 1 0.1299 RCIM_Cubic>]}
       GreenCurve {keys [<at 0 0.8294>,<at 0.3901 0.352>,<at 0.7425 0.6026>,<at 1 0.8908 RCIM_Cubic>]}
       RedCurve {keys [<at 0 1>,<at 0.3901 1>,<at 0.7425 1>,<at 1 1 RCIM_Cubic>]}
      }}
     CurveIndex=&RandomRangeFloat(Minimum=0, Maximum=1)
    )
    Lifetime Mode=::NewEnumerator1
    Sprite Rotation Mode=::NewEnumerator2
    Sprite Size Mode=::NewEnumerator3
    Sprite UV Mode=::NewEnumerator3
    Uniform Sprite Size=&Lerp_Float(
     A=4
     B=7
     Alpha=&FloatFromCurve(
      FloatCurve=DataInterface() {propertyValues {Curve {keys [
          {interpMode=RCIM_Cubic, tangentMode=RCTM_User, time=0, value=0}
          {arriveTangent=1.8215, interpMode=RCIM_Cubic, leaveTangent=1.8215, tangentMode=RCTM_User, time=1, value=1}
         ]}}}
      CurveIndex=&RandomRangeFloat(Minimum=0, Maximum=1)
     )
    )
    UsePositionOffset=false
    Write Material Random=true
   )
   &SkeletalMeshLocation(
    Bone / Socket Index=&SelectIntFromArray(Direct Array Index=$Particles.PartitionParticles.Partition, Int Selection Array=DataInterface() {propertyValues {IntData [0,1,6,7]}})
    Bone Sampling Mode=::NewEnumerator4
    MaskSpawnGroup=false
    Sampled Position Offset=&SelectVectorFromArray(
     Direct Array Index=$Particles.PartitionParticles.Partition
     Vector Selection Array=DataInterface() {propertyValues {FloatData [<0 0 0>, <0 0 0>, <15 0 0>, <-15 0 0>]}}
    )
    Skeletal Mesh=$System.Skeletal Mesh
   )
   &DynamicMaterialParameters(
    Index 0 Param 1=&RandomRangeFloat(Minimum=0.1, Maximum=0.5)
    Module.Index0Param0Write=true
    Module.Index0Param1Write=true
    Module.Index0Param2Write=true
    Module.Index0Param3Write=true
   )
   &ShapeLocation(MaskSpawnGroup=false, Override Seed=false, Sphere Radius=15, Sphere Surface Distribution=0.2, UseOwnerScale=true)
   &CameraOffset(Camera Offset Amount=&RandomRangeFloat(Minimum=65, Maximum=160))
   &InheritVelocity(
    Inherited Velocity Amount Scale=&VectorFromFloat(Value=-0.05)
    Inherited Velocity Speed Limit=300
    Speed Limit=true
    UseSpeedThreshold=false
   )
   &AddVelocity(Velocity=<0 0 -50>, Velocity Speed Scale=&RandomRangeFloat(Minimum=0.05, Maximum=0.4))
   &AddVelocity(
    Module.Use Falloff Distance=false
    Override Seed=false
    Velocity Mode=::NewEnumerator1
    Velocity Origin=$Particles.ShapeLocation.ShapePosition
    Velocity Speed=10
   ) {module=AddVelocity001}
  ]
  particleUpdateScript [
   &ParticleState()
   &InheritSourceMovement(
    Applied Position Scale=&VectorFromFloat(Value=&FloatFromCurve(FloatCurve=DataInterface() {propertyValues {Curve {keys [
          {interpMode=RCIM_Cubic, tangentMode=RCTM_User, time=0.2, value=1}
          {arriveTangent=-0.0567, interpMode=RCIM_Cubic, leaveTangent=-0.0567, tangentMode=RCTM_User, time=1, value=0.8}
         ]}}}))
    ApplyForce=false
    ApplyPosition=true
    ApplyPositionFalloff=false
   )
   &Drag(Drag=2.5, Use Linear Drag=true, Use Rotational Drag=true)
   &SolveForcesAndVelocity(Clamp Velocity=false, Limit Acceleration=false)
   /SetVariables(Particles.DEBUGVel:NiagaraFloat=&VectorLength(Vector=$Engine.Owner.Velocity)) {enabled=false, isSetParameters=true}
   &ScaleColor(
    Scale Alpha=&Multiply_Float(
     A=$System.ADSAlpha
     B=&FloatFromCurve(
      FloatCurve=DataInterface() {propertyValues {Curve {keys [
          {interpMode=RCIM_Cubic, tangentMode=RCTM_User, time=0, value=0}
          {arriveTangent=-2.6389, interpMode=RCIM_Cubic, leaveTangent=-2.6389, tangentMode=RCTM_User, time=0.3, value=1}
          {interpMode=RCIM_Cubic, tangentMode=RCTM_User, time=1, value=0}
         ]}}}
      Scale Curve=&RandomRangeFloat(Minimum=1, Maximum=2)
     )
    )
    Scale RGB=&VectorFromFloat(Value=1)
    ScaleA=true
    ScaleRGB=true
   ) {module=ScaleColor003}
  ]
  rendererClasses [&NiagaraSpriteRendererProperties]
  renderers [{rendererClass=&NiagaraSpriteRendererProperties}]
 }
 Decals {
  emitterUpdateScript [
   &EmitterState()
   &SpawnRate(SpawnRate=&Multiply_Float(A=$System.WaterFXOpacity, B=8), Use Spawn Probability=false)
   &SpawnBurst_Instantaneous(Spawn Count=8, Use Spawn Probability=false, UseLoopCountLimit=true)
  ]
  particleSpawnScript [
   &PartitionParticles(Number Of Partitions=8)
   &InitializeParticle(
    Color=<rgba 1 1 1 0>
    Lifetime=1.5
    Sprite Size Mode=::NewEnumerator3
    Uniform Sprite Size=&Multiply_Float(
     A=25
     B=&SelectFloatFromArray(
      Direct Array Index=$Particles.PartitionParticles.Partition
      Float Selection Array=DataInterface() {propertyValues {FloatData [1,1,0.7,0.7,2,2,1,1]}}
     )
    )
    UsePositionOffset=false
    Write Material Random=true
   )
   &CameraOffset(Camera Offset Amount=&SelectFloatFromArray(
     Direct Array Index=$Particles.PartitionParticles.Partition
     Float Selection Array=DataInterface() {propertyValues {FloatData [15,15,15,15,15,15,50,50]}}
    ))
  ]
  particleUpdateScript [
   &ParticleState()
   &SkeletalMeshLocation(
    Bone / Socket Index=$Particles.PartitionParticles.Partition
    Bone Sampling Mode=::NewEnumerator4
    MaskSpawnGroup=false
    Sampled Position Offset=&SelectVectorFromArray(
     Direct Array Index=$Particles.PartitionParticles.Partition
     Vector Selection Array=DataInterface() {propertyValues {FloatData [<0 0 0>, <0 0 0>, <0 0 0>, <0 0 0>, <-6 0 -5>, <6 0 5>, <5 0 0>, <-5 0 0>]}}
    )
    Skeletal Mesh=$System.Skeletal Mesh
   )
   /SetVariables(Particles.Color=&MakeLinearColorFromVectorAndFloat(
     "Vector (RGB)"=&Multiply_VectorByFloat(Vector=&MakeVectorFromLinearColorRGB(Color=<rgba 1 0.5307 0.12 1>), Float=1)
     "Float (Alpha)"=&Multiply_Float(
      A=$System.WaterFXOpacity
      B=&Multiply_Float(
       A=$System.ADSAlpha
       B=&Multiply_Float(
        A=&FloatFromCurve(
         FloatCurve=DataInterface() {propertyValues {Curve {keys [
             {interpMode=RCIM_Cubic, tangentMode=RCTM_User, time=0, value=0}
             {interpMode=RCIM_Cubic, tangentMode=RCTM_User, time=0.3, value=1}
             {interpMode=RCIM_Cubic, tangentMode=RCTM_User, time=1, value=0}
            ]}}}
         Scale Curve=&FN_PickFloatDependOnPlatform(TargetPlatformActive=1, TargetPlatformNoActive=0.3)
        )
        B=&SelectFloatFromArray(
         Direct Array Index=$Particles.PartitionParticles.Partition
         Float Selection Array=DataInterface() {propertyValues {FloatData [3,3,1,1,1,1,3,3]}}
        )
       )
      )
     )
    )) {isSetParameters=true}
   &DynamicMaterialParameters(
    Index 0 Param 1=&Lerp_Float(
     A=2.5
     B=4
     Alpha=&FloatFromCurve(FloatCurve=DataInterface() {propertyValues {Curve {keys [{interpMode=RCIM_Cubic, tangentMode=RCTM_User, time=0, value=0}, {interpMode=RCIM_Cubic, tangentMode=RCTM_User, time=1, value=1}]}}})
    )
    Index 0 Param 2=1
    Module.Index0Param0Write=true
    Module.Index0Param1Write=false
    Module.Index0Param2Write=false
    Module.Index0Param3Write=false
   ) {module=DynamicMaterialParameters001}
  ]
  rendererClasses [&NiagaraSpriteRendererProperties]
  renderers [
   {rendererClass=&NiagaraSpriteRendererProperties}
   {rendererClass=&NiagaraSpriteRendererProperties, rendererIndex=1}
  ]
 }
 MainRibbonOnLocalClient {
  emitterUpdateScript [
   /SetVariables(Emitter.EmitterRandom:NiagaraFloat=&RandomRangeFloat(Minimum=0, Maximum=1)) {isSetParameters=true}
   &EmitterState(
    Enable Distance Culling=true
    Inactive Response=::NewEnumerator1
    Life Cycle Mode=::NewEnumerator1
    Loop Behavior=::NewEnumerator1
    Loop Delay=0.2
    Loop Duration Mode=::NewEnumerator1
    MaxDistance=9000
    Module.bUseMaxDistance=true
    Module.bUseMinDistance=false
    Scalability Mode=::NewEnumerator1
    UseLoopDelay=false
   )
   &SpawnBurst_Instantaneous(Spawn Count=&Multiply_Int(A=4, B=10), Use Spawn Probability=false, UseLoopCountLimit=false)
  ]
  particleSpawnScript [
   &PartitionParticles(
    Number Of Partitions=4
    Particles Per Partition=10
    Partition Mode:ENiagara_PartitionMode=::NewEnumerator1
    Write Partition Index to Ribbon Link Order:ENiagara_AttributeSamplingApplyOutput=::NewEnumerator0
   )
   &InitializeParticle(Ribbon Width=15, Ribbon Width Mode=::NewEnumerator1, Sprite Size Mode=::NewEnumerator3, UsePositionOffset=false, Write Material Random=true)
   /SetVariables(Particles.ArmLegAlpha:NiagaraFloat=&MakeCustomFloatFromBool(
     Bool=&SetBoolByIntComparison(A=&SpawnGroup(), B=2, Comparison Type=::NewEnumerator2)
     True Float=0
     False Float=1
    )) {isSetParameters=true}
   /SetVariables(Particles.Lifetime=&Lerp_Float(
     A=&Lerp_Float(A=0.04, B=0.08, Alpha=$System.NormalizedVelocityRange)
     B=&Lerp_Float(A=0.1, B=0.12, Alpha=$System.NormalizedVelocityRange)
     Alpha=$Particles.ArmLegAlpha
    )) {isSetParameters=true}
   /SetVariables(Particles.RandomWave:NiagaraFloat=&Waveform(
     Global Phase=&Add_Float(A=$System.Age, B=&Multiply_FloatByInt(Float=0.73, Integer=&SpawnGroup()))
     "Waveform [1]"=::NewEnumerator12
     "[1] Frequency"=5
     "[1] Amplitude Min/Max"=<0 1>
    )) {isSetParameters=true}
   &SkeletalMeshLocation(
    Bone / Socket Index=&SelectIntFromArray(Direct Array Index=$Particles.PartitionParticles.Partition, Int Selection Array=DataInterface() {propertyValues {IntData [6,7,0,1]}})
    Bone Sampling Mode=::NewEnumerator4
    Interpolate Mesh Sampling=false
    MaskSpawnGroup=false
    Orientation Sampling=::NewEnumerator1
    Sampled Position Offset=&SelectVectorFromArray(Direct Array Index=$Particles.PartitionParticles.Partition, Vector Selection Array=$System.SkelOffset)
    Skeletal Mesh=$System.Skeletal Mesh
   )
   /SetVariables(Particles.InitialNormalizedVelocityRange:NiagaraFloat=$System.NormalizedVelocityRange) {isSetParameters=true}
  ]
  particleUpdateScript [
   &ParticleState(Kill Particles When Lifetime Has Elapsed=false)
   &SkeletalMeshLocation(
    Bone / Socket Index=&SelectIntFromArray(Direct Array Index=$Particles.PartitionParticles.Partition, Int Selection Array=DataInterface() {propertyValues {IntData [6,7,0,1]}})
    Bone Sampling Mode=::NewEnumerator4
    MaskSpawnGroup=false
    Orientation Sampling=::NewEnumerator1
    Position Sampling=::NewEnumerator2
    Sampled Position Offset=&SelectVectorFromArray(Direct Array Index=$Particles.PartitionParticles.Partition, Vector Selection Array=$System.SkelOffset)
    Skeletal Mesh=$System.Skeletal Mesh
   ) {module=SkeletalMeshLocation001}
   /SetVariables(Particles.GoalPos:NiagaraPosition=&AddVectorToPosition(
     Position=$Particles.SkeletalMeshLocation001.SampledPosition
     Vector=&Multiply_VectorByFloat(
      Vector=&Lerp_Vector(
       A=<0 0 0>
       B=&Multiply_VectorByFloat(
        Vector=&Lerp_Vector(
         A=<0 0 0>
         B=&Multiply_VectorByFloat(
          Vector=&NormalizeVector(
           Vector To Normalize=/NiagaraFluids/DynamicInput/MaxVectorByLength(
            A:Vector3f=<0 -1 0>
            B:Vector3f=&Multiply_VectorByFloat(
             Vector=&TransformVector(Vector=$Engine.Owner.Velocity, Source Space=::World, Destination Space=::Local)
             Float=-1
            )
           )
           Safe Normalize=true
           Safe Fallback Threshold=0
           Safe Fallback Vector=<0 0 0.001>
          )
          Float=&SelectFloatFromArray(
           Direct Array Index=$Particles.PartitionParticles.Partition
           Float Selection Array=DataInterface() {propertyValues {FloatData [220,220,120,120]}}
          )
         )
         Alpha=$Particles.PartitionParticles.Normalized Index In Partition
        )
        Float=&MakeCustomFloatFromBool(Bool=$System.IsThirdPersonPlayerWatching, True Float=1.8, False Float=1.2)
       )
       Alpha=$System.NormalizedVelocityRange
      )
      Float=$System.WaterFXOpacity
     )
    )) {isSetParameters=true}
   &InheritSourceMovement(
    Applied Position Scale=&VectorFromFloat(Value=&Multiply_Float(
      A=-0.5
      B=&FloatFromCurve(
       FloatCurve=DataInterface() {propertyValues {Curve {keys [
           {arriveTangent=-0.0194, interpMode=RCIM_Cubic, leaveTangent=-0.0194, tangentMode=RCTM_User, time=0.2, value=0}
           {arriveTangent=1.6438, interpMode=RCIM_Cubic, leaveTangent=1.6438, tangentMode=RCTM_User, time=1, value=1}
          ]}}}
       CurveIndex=$Particles.PartitionParticles.Normalized Index In Partition
      )
     ))
    ApplyForce=false
    ApplyPosition=true
    ApplyPositionFalloff=false
    Source Velocity:Vector3f=&TransformVector(Vector=$Engine.Owner.Velocity, Source Space=::World, Destination Space=::Local)
   )
   &ScaleColor(
    Scale Alpha=&FloatFromCurve(
     FloatCurve=DataInterface() {propertyValues {Curve {keys [
         {arriveTangent=0.0106, interpMode=RCIM_Cubic, leaveTangent=0.0106, tangentMode=RCTM_User, time=0, value=0}
         {arriveTangent=0.0717, interpMode=RCIM_Cubic, leaveTangent=0.0717, tangentMode=RCTM_User, time=0.5, value=1}
         {interpMode=RCIM_Cubic, tangentMode=RCTM_User, time=1, value=0}
        ]}}}
     CurveIndex=$Particles.PartitionParticles.Normalized Index In Partition
     Scale Curve=&Lerp_Float(A=0, B=0.75, Alpha=$System.NormalizedVelocityRange)
    )
    Scale RGB=&MakeVectorFromLinearColorRGB(Color=&ColorFromCurve(
      ColorCurve=DataInterface() {propertyValues {
        AlphaCurve {keys [<at 0 1 RCIM_Cubic>,<at 1 0 RCIM_Cubic>]}
        BlueCurve {keys [<at 0 0.2 RCIM_Cubic>,<at 0.2632 0.02 RCIM_Cubic>]}
        GreenCurve {keys [<at 0 4.61 RCIM_Cubic>,<at 0.2632 0.3793 RCIM_Cubic>]}
        RedCurve {keys [<at 0 10 RCIM_Cubic>,<at 0.2632 1 RCIM_Cubic>]}
       }}
      CurveIndex=$Particles.PartitionParticles.Normalized Index In Partition
     ))
    ScaleA=true
    ScaleRGB=true
   )
   /SetVariables(Particles.RibbonWidth=&Lerp_Float(
     A=5
     B=20
     Alpha=&FloatFromCurve(
      FloatCurve=DataInterface() {propertyValues {Curve {keys [
          {arriveTangent=4.8892, interpMode=RCIM_Cubic, leaveTangent=4.8892, tangentMode=RCTM_User, time=0, value=0.2}
          {interpMode=RCIM_Cubic, tangentMode=RCTM_User, time=0.2, value=1}
          {interpMode=RCIM_Cubic, tangentMode=RCTM_User, time=0.7, value=0.2}
         ]}}}
      CurveIndex=$Particles.PartitionParticles.Normalized Index In Partition
     )
    )) {isSetParameters=true}
   &DynamicMaterialParameters(
    Index 0 Param 1=/Game/Effects/Niagara/DynamicInputs/Random/FN_SeededRandomRangeFloat(Minimum:NiagaraFloat=0, Maximum:NiagaraFloat=1, Random Seed:NiagaraInt32=&SpawnGroup())
    Index 0 Param 2=0.25
    Index 0 Param 3=&SelectFloatFromArray(Direct Array Index=$Particles.PartitionParticles.Partition, Float Selection Array=DataInterface() {propertyValues {FloatData [1,0,1,0]}})
    Module.Index0Param0Write=true
    Module.Index0Param1Write=true
    Module.Index0Param2Write=true
    Module.Index0Param3Write=true
   )
   &ScaleColor(
    Scale Alpha=&Multiply_Float(
     A=&FloatFromCurve(
      FloatCurve=DataInterface() {propertyValues {Curve {keys [{interpMode=RCIM_Cubic, tangentMode=RCTM_User, time=0, value=1}, {interpMode=RCIM_Cubic, tangentMode=RCTM_User, time=1, value=0.5}]}}}
      CurveIndex=$System.Age
      Scale Curve=1.5
     )
     B=&Multiply_Float(A=$System.WaterFXOpacity, B=&Lerp_Float(A=0.7, B=1, Alpha=$Particles.ArmLegAlpha))
    )
    ScaleA=true
    ScaleRGB=true
   ) {module=ScaleColor001}
   &SolveForcesAndVelocity(Clamp Velocity=false, Limit Acceleration=false)
   /SetVariables(Particles.Position=&Lerp_Vector(
     A=Linked() {linkedVariable {name=Particles.GoalPos, type=NiagaraPosition}}
     B=&Lerp_Vector(
      A=Linked() {linkedVariable {name=Particles.Position, type=NiagaraPosition}}
      B=Linked() {linkedVariable {name=Particles.GoalPos, type=NiagaraPosition}}
      Alpha=&Multiply_Float(A=$Engine.DeltaTime, B=2)
     )
     Alpha=&FloatFromCurve(
      FloatCurve=DataInterface() {propertyValues {Curve {keys [
          {arriveTangent=3.4377, interpMode=RCIM_Cubic, leaveTangent=3.4377, tangentMode=RCTM_User, time=0, value=0}
          {interpMode=RCIM_Cubic, tangentMode=RCTM_User, time=1, value=1}
         ]}}}
      CurveIndex=$Particles.PartitionParticles.Normalized Index In Partition
     )
    )) {isSetParameters=true}
   &ScaleColor(
    Scale Alpha=&MakeCustomFloatFromBool(Bool=$System.IsThirdPersonPlayerWatching, True Float=2, False Float=1)
    ScaleA=true
    ScaleRGB=false
   ) {module=ScaleColor002}
  ]
  rendererClasses [&NiagaraRibbonRendererProperties]
  renderers [{rendererClass=&NiagaraRibbonRendererProperties}]
 }
 Particulate {
  emitterUpdateScript [
   &EmitterState()
   &SpawnRate(SpawnRate=20, Use Spawn Probability=false) {enabled=false, module=SpawnRate001}
   &SpawnRate(
    Spawn Probability=0.9
    SpawnRate=&Multiply_Float(
     A=$System.WaterFXOpacity
     B=&Multiply_Float(
      A=50
      B=&FloatFromCurve(
       FloatCurve=DataInterface() {propertyValues {Curve {keys [{interpMode=RCIM_Cubic, tangentMode=RCTM_User, time=0.2, value=0}, {interpMode=RCIM_Cubic, tangentMode=RCTM_User, time=0.35, value=1}]}}}
       CurveIndex=$Emitter.Age
       Scale Curve=&FloatFromCurve(
        FloatCurve=DataInterface() {propertyValues {Curve {keys [{interpMode=RCIM_Cubic, tangentMode=RCTM_User, time=0.1, value=0.2}, {interpMode=RCIM_Cubic, tangentMode=RCTM_User, time=1, value=1}]}}}
        CurveIndex=&NormalizeFloat(Value=&VectorLength(Vector=$Engine.Owner.Velocity), Min=50, Max=500)
       )
      )
     )
    )
    Use Spawn Probability=false
   )
  ]
  particleSpawnScript [
   &FN_SetSpawnGroup(Particles Per Group:NiagaraInt32=$Emitter.SpawnGroup, Pattern=::NewEnumerator0) {enabled=false}
   &PartitionParticles(Number Of Partitions=8)
   /SetVariables(Particles.ParticleGroup:NiagaraInt32=/Niagara/DynamicInputs/Divide/Divide_Int(A:NiagaraInt32=$Particles.PartitionParticles.Partition, B:NiagaraInt32=4)) {isSetParameters=true}
   /SetVariables(Particles.VisibilityTag=$Particles.ParticleGroup) {isSetParameters=true}
   &InitializeParticle(
    Color=<rgba 1 0.4306 0 1>
    Lifetime=&MaskFloat(
     Mask Parameter=$Particles.ParticleGroup
     Mask=0
     Masked Float=&RandomRangeFloat(Minimum=0.7, Maximum=0.9)
     Passthrough Float=&RandomRangeFloat(Minimum=1, Maximum=1.3)
    )
    Sprite Rotation Angle=&MaskFloat(
     Mask Parameter=$Particles.ParticleGroup
     Mask=0
     Masked Float=&RandomRangeFloat(Minimum=0, Maximum=1)
     Passthrough Float=0
    )
    Sprite Rotation Mode=::NewEnumerator1
    Sprite Size Mode=::NewEnumerator3
    Uniform Sprite Size=&Multiply_Float(
     A=&RandomRangeFloat(Minimum=20, Maximum=35)
     B=&SelectFloatFromArray(Direct Array Index=$Particles.ParticleGroup, Float Selection Array=DataInterface() {propertyValues {FloatData [1,1.7]}})
    )
    UsePositionOffset=false
    Write Material Random=true
   )
   &CameraOffset(Camera Offset Amount=&FN_PickFloatDependOnPlatform(TargetPlatformActive=9, TargetPlatformNoActive=-15))
   &SkeletalMeshLocation(
    Bone / Socket Index=&SelectIntFromArray(
     Direct Array Index=&ModuloInt(A=$Particles.PartitionParticles.Partition, B=4)
     Int Selection Array=DataInterface() {propertyValues {IntData [0,1,4,5]}}
    )
    Bone Sampling Mode=::NewEnumerator4
    MaskSpawnGroup=false
    Sampled Position Offset=&SelectVectorFromArray(
     Direct Array Index=&ModuloInt(A=$Particles.PartitionParticles.Partition, B=4)
     Vector Selection Array=DataInterface() {propertyValues {FloatData [<0 0 0>, <0 0 0>, <-15 0 0>, <15 0 0>]}}
    )
    Skeletal Mesh=$System.Skeletal Mesh
   )
   &InheritVelocity(
    Inherited Velocity Amount Scale=&VectorFromFloat(Value=&MaskFloat(Mask Parameter=$Particles.ParticleGroup, Mask=0, Masked Float=0.4, Passthrough Float=0.6))
    Inherited Velocity Speed Limit=1000
    Source Speed Threshold=400
    Speed Limit=true
    UseSpeedThreshold=false
   ) {module=InheritVelocity001}
   &SpriteFacingAndAlignment(Facing Coordinate Space=::World, Sprite Facing=<0 0 1>, WriteAlignment=false, WriteFacing=true)
   &ShapeLocation(MaskSpawnGroup=false, Override Seed=false, Sphere Radius=2, UseOwnerScale=true) {enabled=false}
  ]
  particleUpdateScript [
   &ParticleState()
   &SkeletalMeshLocation(
    Bone / Socket Index=&SelectIntFromArray(
     Direct Array Index=&ModuloInt(A=$Particles.PartitionParticles.Partition, B=4)
     Int Selection Array=DataInterface() {propertyValues {IntData [0,1,4,5]}}
    )
    Bone Sampling Mode=::NewEnumerator4
    MaskSpawnGroup=false
    Position Sampling=::NewEnumerator1
    Sampled Position Offset=&SelectVectorFromArray(
     Direct Array Index=&ModuloInt(A=$Particles.PartitionParticles.Partition, B=4)
     Vector Selection Array=DataInterface() {propertyValues {FloatData [<0 0 0>, <0 0 0>, <-25 0 0>, <25 0 0>]}}
    )
    Skeletal Mesh=$System.Skeletal Mesh
    Source Movement Scale:Vector3f=&VectorFromFloat(Value=&FloatFromCurve(
      FloatCurve=DataInterface() {propertyValues {Curve {keys [
          {interpMode=RCIM_Cubic, tangentMode=RCTM_User, time=0.2, value=1}
          {arriveTangent=-1.3739, interpMode=RCIM_Cubic, leaveTangent=-1.3739, tangentMode=RCTM_User, time=0.8, value=0}
         ]}}}
      Scale Curve=&MaskFloat(Mask Parameter=$Particles.ParticleGroup, Mask=0, Masked Float=0.85, Passthrough Float=1)
     ))
   ) {enabled=false, module=SkeletalMeshLocation001}
   &InheritVelocity(
    Inherited Velocity Amount Scale=&VectorFromFloat(Value=&FloatFromCurve(
      FloatCurve=DataInterface() {propertyValues {Curve {keys [{interpMode=RCIM_Cubic, tangentMode=RCTM_User, time=0.5, value=0}, {interpMode=RCIM_Cubic, tangentMode=RCTM_User, time=1, value=1}]}}}
      Scale Curve=0.25
     ))
    Inherited Velocity Speed Limit=300
    Speed Limit=true
    UseSpeedThreshold=false
   ) {enabled=false}
   &GravityForce(Gravity=&MakeVector(
     X=0
     Y=0
     Z=&Multiply_Float(
      A=&FloatFromCurve(FloatCurve=DataInterface() {propertyValues {Curve {keys [{interpMode=RCIM_Cubic, tangentMode=RCTM_User, time=0, value=0}, {interpMode=RCIM_Cubic, tangentMode=RCTM_User, time=0.4, value=1}]}}})
      B=&Multiply_Float(
       A=&RandomRangeFloat(Minimum=-80, Maximum=-100)
       B=&MaskFloat(Mask Parameter=$Particles.ParticleGroup, Mask=0, Masked Float=2, Passthrough Float=1)
      )
     )
    ))
   &ScaleSpriteSize(
    Scale Sprite Size Mode=::NewEnumerator0
    Uniform Scale Factor=&MaskFloat(
     Mask Parameter=$Particles.ParticleGroup
     Mask=0
     Masked Float=&FloatFromCurve(FloatCurve=DataInterface() {propertyValues {Curve {keys [
          {arriveTangent=1.3922, interpMode=RCIM_Cubic, leaveTangent=1.3922, leaveTangentWeight=0.3665, tangentMode=RCTM_User, time=0, value=0.33}
          {interpMode=RCIM_Cubic, tangentMode=RCTM_User, time=1, value=1}
         ]}}})
     Passthrough Float=&FloatFromCurve(FloatCurve=DataInterface() {propertyValues {Curve {keys [
          {arriveTangent=0.2245, interpMode=RCIM_Cubic, leaveTangent=0.2245, tangentMode=RCTM_User, time=0, value=0.4}
          {arriveTangent=0.0001, interpMode=RCIM_Cubic, leaveTangent=0.0001, tangentMode=RCTM_User, time=1, value=1}
         ]}}})
    )
   )
   &Drag(Drag=9, Use Linear Drag=true, Use Rotational Drag=true)
   &InheritSourceMovement(
    Applied Position Scale=&VectorFromFloat(Value=&FloatFromCurve(
      FloatCurve=DataInterface() {propertyValues {Curve {keys [
          {arriveTangent=-0.1589, interpMode=RCIM_Cubic, leaveTangent=-0.1589, tangentMode=RCTM_User, time=0.1, value=1}
          {arriveTangent=-1.7422, interpMode=RCIM_Cubic, leaveTangent=-1.7422, tangentMode=RCTM_User, time=1, value=0.5}
         ]}}}
      Scale Curve=&MaskFloat(Mask Parameter=$Particles.ParticleGroup, Mask=0, Masked Float=0.65, Passthrough Float=0.85)
     ))
    ApplyForce=false
    ApplyPosition=true
    ApplyPositionFalloff=false
   )
   &InheritSourceMovement(
    Applied Position Scale=&VectorFromFloat(Value=&FloatFromCurve(
      FloatCurve=DataInterface() {propertyValues {Curve {keys [
          {arriveTangent=-0.1589, interpMode=RCIM_Cubic, leaveTangent=-0.1589, tangentMode=RCTM_User, time=0.1, value=1}
          {arriveTangent=-1.7422, interpMode=RCIM_Cubic, leaveTangent=-1.7422, tangentMode=RCTM_User, time=1, value=0.5}
         ]}}}
      Scale Curve=0.3
     ))
    ApplyForce=false
    ApplyPosition=true
    ApplyPositionFalloff=false
   ) {enabled=false, module=InheritSourceMovement001}
   &SolveForcesAndVelocity(Clamp Velocity=false, Limit Acceleration=false)
   &ScaleColor(
    Scale Alpha=&Lerp_Float(
     A=1
     B=&FloatFromCurve(
      FloatCurve=DataInterface() {propertyValues {Curve {keys [
          {interpMode=RCIM_Cubic, tangentMode=RCTM_User, time=0, value=1}
          {arriveTangentWeight=0.0471, interpMode=RCIM_Cubic, leaveTangentWeight=0.33, tangentMode=RCTM_User, time=0.2, value=0}
          {arriveTangentWeight=0.3202, interpMode=RCIM_Cubic, leaveTangentWeight=0.1067, tangentMode=RCTM_User, time=0.8, value=0}
          {arriveTangent=7.803, interpMode=RCIM_Cubic, leaveTangent=7.803, tangentMode=RCTM_User, time=1, value=1}
         ]}}}
      Scale Curve=0.1
     )
     Alpha=$System.ADSAlpha
    )
    Scale RGB=&Multiply_Vector(
     A=&VectorFromFloat(Value=&Lerp_Float(
       A=1500
       B=3000
       Alpha=&FloatFromCurve(
        FloatCurve=DataInterface() {propertyValues {Curve {keys [
            {interpMode=RCIM_Cubic, tangentMode=RCTM_User, time=0, value=0}
            {arriveTangent=1, interpMode=RCIM_Cubic, leaveTangent=1, time=0.7, value=0.2}
            {arriveTangentWeight=0.5245, interpMode=RCIM_Cubic, tangentMode=RCTM_User, time=1, value=1}
           ]}}}
        CurveIndex=&RandomRangeFloat(Minimum=0, Maximum=1)
       )
      ))
     B=&VectorFromFloat(Value=&FloatFromCurve(FloatCurve=DataInterface() {propertyValues {Curve {keys [{interpMode=RCIM_Cubic, tangentMode=RCTM_User, time=0, value=1}, {interpMode=RCIM_Cubic, tangentMode=RCTM_User, time=0.85, value=0.1}]}}}))
    )
    ScaleA=true
    ScaleRGB=true
   )
   &DynamicMaterialParameters(
    Index 0 Param 1=&Lerp_Float(
     A=0.02
     B=0.4
     Alpha=&FloatFromCurve(FloatCurve=DataInterface() {propertyValues {Curve {keys [
          {interpMode=RCIM_Cubic, tangentMode=RCTM_User, time=0.6, value=0}
          {arriveTangent=5.3655, interpMode=RCIM_Cubic, leaveTangent=5.3655, tangentMode=RCTM_User, time=1, value=1}
         ]}}})
    )
    Index 0 Param 2=&RandomRangeFloat(Minimum=2, Maximum=4)
    Index 0 Param 3=&RandomRangeFloat(Minimum=0, Maximum=1)
    Index 0 Param 4=&Lerp_Float(
     A=0.2
     B=0.4
     Alpha=&FloatFromCurve(
      FloatCurve=DataInterface() {propertyValues {Curve {keys [
          {interpMode=RCIM_Cubic, tangentMode=RCTM_User, time=0, value=0}
          {arriveTangent=2.573, interpMode=RCIM_Cubic, leaveTangent=2.573, tangentMode=RCTM_User, time=1, value=1}
         ]}}}
      CurveIndex=&RandomRangeFloat(Minimum=0, Maximum=1)
     )
    )
    Module.Index0Param0Write=true
    Module.Index0Param1Write=true
    Module.Index0Param2Write=true
    Module.Index0Param3Write=true
   )
   /SetVariables(
    Particles.ColorGoldDust:LinearColor=&MultiplyLinearColor(
     Linear Color=<rgba 1 1 1 1>
     Scale Factor=&MakeVector4(
      X=1
      Y=1
      Z=1
      W=&Multiply_Float(
       A=$System.ADSAlpha
       B=&FloatFromCurve(
        FloatCurve=DataInterface() {propertyValues {Curve {keys [
            {interpMode=RCIM_Cubic, tangentMode=RCTM_User, time=0, value=0}
            {arriveTangent=-1.0029, interpMode=RCIM_Cubic, leaveTangent=-1.0029, tangentMode=RCTM_User, time=0.35, value=1}
            {arriveTangent=-0.1502, interpMode=RCIM_Cubic, leaveTangent=-0.1502, tangentMode=RCTM_User, time=1, value=0}
           ]}}}
        Scale Curve=&Multiply_Float(
         A=&RandomRangeFloat(Minimum=0.6, Maximum=0.9)
         B=&FN_PickFloatDependOnPlatform(TargetPlatformActive=1, TargetPlatformNoActive=0.3)
        )
       )
      )
     )
    )
    Particles.DynMatGoldDust:Vector4f=&MakeVector4(
     X=&RandomRangeFloat(Minimum=-0.05, Maximum=0.05)
     Y=&RandomRangeFloat(Minimum=0, Maximum=32)
     Z=0
     W=0
    )
   ) {isSetParameters=true}
  ]
  rendererClasses [&NiagaraSpriteRendererProperties]
  renderers [
   {rendererClass=&NiagaraSpriteRendererProperties}
   {rendererClass=&NiagaraSpriteRendererProperties, rendererIndex=1}
  ]
 }
}
summary {
 systemName=NS_GoldRush_Idle
 emitters [
  {emitterName=Decals, rendererClasses [&NiagaraSpriteRendererProperties]}
  {emitterName=Particulate, rendererClasses [&NiagaraSpriteRendererProperties]}
  {emitterName=Bokeh, rendererClasses [&NiagaraSpriteRendererProperties]}
  {emitterName=MainRibbonOnLocalClient, rendererClasses [&NiagaraRibbonRendererProperties]}
 ]
 userVariables [
  User.bIsFrontEndPreview {type=NiagaraBool}
  User.Debug {type=NiagaraBool}
  User.FadeOut {type=NiagaraBool}
  User.RarityLevel {type=NiagaraInt32}
  User.SystemScale {type=NiagaraFloat, defaultValue=1, description=""}
  User.WaterFade {type=NiagaraFloat, defaultValue=0, description=""}
 ]
}
